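# Sound cue for the Skyfire end room reveal music, playing a main room completion transition stinger.
# A fixed delay of just under a second feeds a single wave player, so the stinger starts slightly after the cue triggers.
# The cue uses the LTM in-game music sound class with high priority that bypasses volume scaling, keeps playing when silent, and has a slightly boosted volume.
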SoundCue /Skyfire/Sound/Music/Skyfire_MX_End_Rm_Reveal_Cue
root=Delay_0
nodes [
 {id=Delay_0, children [WavePlayer_0], properties {delayMax=0.95, delayMin=0.95}}
 {id=WavePlayer_0, properties {soundWaveAssetPtr=&/Skyfire_MX_Main_Rm_Complete_Transition_STNG}}
]
properties {
 bBypassVolumeScaleForPriority=true
 maxDistance=2097152
 priority=100
 soundClassObject=/Game/Sounds/FortSoundClassesAndModes/Fort_LTM_BR_InGame_Music
 virtualizationMode=PlayWhenSilent
 volumeMultiplier=1.2
}
