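# High-priority looping music cue for the Skyfire main room, routed to the LTM in-game ambient 2D music sound class.
# The first two quality levels mix Base, Loot and Warning loops, whose volumes follow same-named parameters defaulting to zero; the other two use a single composite loop.
# An enveloper eases the loop volume down over ten seconds, and a one-shot start stinger joins only at the first quality level.
# The cue keeps playing when silent.
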SoundCue /Skyfire/Sound/Music/Skyfire_MX_Main_Rm_Loop_Cue
root=Mixer_1
nodes [
 {
  id=Enveloper_0
  children [QualityLevel_0]
  properties {
   pitchCurve {editorCurveData {keys [{time=0}]}}
   volumeCurve {editorCurveData {keys [{arriveTangent=-0.0843, interpMode=RCIM_Cubic, leaveTangent=-0.0843, tangentMode=RCTM_User, time=0, value=1}, <at 10 0.7>]}}
  }
 }
 {id=Mixer_0, children [ModulatorContinuous_1, ModulatorContinuous_0, ModulatorContinuous_2], properties {inputVolume [1,1,1]}}
 {id=Mixer_1, children [QualityLevel_1, Enveloper_0], properties {inputVolume [0.85,1]}}
 {id=ModulatorContinuous_0, children [WavePlayer_0], properties {volumeModulationParams {default=0, parameterName=Loot}}}
 {id=ModulatorContinuous_1, children [WavePlayer_1], properties {volumeModulationParams {default=0, parameterName=Base}}}
 {id=ModulatorContinuous_2, children [WavePlayer_2], properties {volumeModulationParams {default=0, parameterName=Warning}}}
 {id=QualityLevel_0, children [Mixer_0,Mixer_0,WavePlayer_4,WavePlayer_4]}
 {id=QualityLevel_1, children [WavePlayer_3,null,null,null]}
 {id=WavePlayer_0, properties {bLooping=true, soundWaveAssetPtr=&/Skyfire_MX_Main_Rm_Loot_Layer_Loop}}
 {id=WavePlayer_1, properties {bLooping=true, soundWaveAssetPtr=&/Skyfire_MX_Main_Rm_Base_Layer_Loop}}
 {id=WavePlayer_2, properties {bLooping=true, soundWaveAssetPtr=&/Skyfire_MX_Main_Rm_Warning_Layer_Loop}}
 {id=WavePlayer_3, properties {soundWaveAssetPtr=&/Skyfire_MX_Main_Rm_Base_Layer_Start_STNG}}
 {id=WavePlayer_4, properties {bLooping=true, soundWaveAssetPtr=&/Skyfire_MX_Main_Rm_Quality_Comp_Loop}}
]
properties {
 bBypassVolumeScaleForPriority=true
 maxDistance=2097152
 priority=100
 soundClassObject=/Game/Sounds/FortSoundClassesAndModes/Fort_LTM_BR_InGame_Music_Ambient_2D
 virtualizationMode=PlayWhenSilent
 volumeMultiplier=2.4
}
