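# Skyfire music Sound Cue for the starting room reveal stinger.
# A single Wave Player plays the reveal stinger wave directly as the cue root, with a volume boost.
# It routes to the LTM in-game music sound class at high priority that bypasses volume scaling, and keeps playing when silent through virtualization.
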SoundCue /Skyfire/Sound/Music/Skyfire_MX_Starting_Rm_Reveal_Cue
root=WavePlayer_4
nodes [{id=WavePlayer_4, properties {soundWaveAssetPtr=&/Skyfire_MX_Starting_Rm_Reveal_STNG}}]
properties {
 bBypassVolumeScaleForPriority=true
 maxDistance=2097152
 priority=100
 soundClassObject=/Game/Sounds/FortSoundClassesAndModes/Fort_LTM_BR_InGame_Music
 virtualizationMode=PlayWhenSilent
 volumeMultiplier=1.35
}
