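# Sound cue that plays a Skyfire music stinger, named for a starting room transition.
# A single wave player feeds the output directly, with no randomization, mixing or other processing nodes.
# It routes to the Fortnite limited-time-mode in-game music sound class with boosted volume, high priority that ignores volume scaling, and play-when-silent virtualization.
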SoundCue /Skyfire/Sound/Music/Skyfire_MX_Starting_Rm_Transition_Cue
root=WavePlayer_0
nodes [{id=WavePlayer_0, properties {soundWaveAssetPtr=&/Skyfire_MX_Starting_Rm_Transition_STNG}}]
properties {
 bBypassVolumeScaleForPriority=true
 maxDistance=2097152
 priority=100
 soundClassObject=/Game/Sounds/FortSoundClassesAndModes/Fort_LTM_BR_InGame_Music
 virtualizationMode=PlayWhenSilent
 volumeMultiplier=1.65
}
